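; Blueprint for the RustBucket Small Silo A environment prop, apparently a building-type actor with a static mesh.
; Its construction script runs the parent version, then fetches the building texture data array.
; If that array is non-empty, it sets slot 0 of the silo mesh to a dynamic material instance that BTechArtGlobalLibrary returns for the StaticMeshComponent variable.
; BeginPlay, overlap and Tick events are empty.
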(event EventBeginPlay)

(event Collision|EventActorBeginOverlap
  (params (OtherActor :type Actor%20Object%20Reference)))

(event EventTick
  (params (DeltaSeconds :type Float%20%28single-precision%29)))

(fn ConstructionScript
  (|CallParentFunction)
  (bind _outtexturedataarray (Building|GetBuildingTextureData))
  (if (> (Utilities|Array|Length _outtexturedataarray) 0)
    (Rendering|Material|SetMaterial (Variables|Default|GetSM_RustBucket_SmallSilo_A) 0 (Class|BTechArtGlobalLibrary|GetDynamicMaterialInstance (Variables|Debug|GetStaticMeshComponent)))))
