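; Fortnite building prop Blueprint for the Helios RustBucket small silo A, which syncs a secondary silo mesh with the building's texture data.
; After the parent construction script, any first texture data entry is copied to index 1, and its tint colors the secondary mesh.
; The main static mesh's dynamic material instance is then assigned to that mesh's first material slot.
; BeginPlay, overlap and Tick events are empty.
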(event EventBeginPlay)

(event Collision|EventActorBeginOverlap
  (params (OtherActor :type Actor%20Object%20Reference)))

(event EventTick
  (params (DeltaSeconds :type Float%20%28single-precision%29)))

(fn ConstructionScript
  (bind _sm_rustbucket_smallsilo_a (Variables|Default|GetSM_RustBucket_SmallSilo_A))
  (|CallParentFunction)
  (bind _outtexturedataarray (Building|GetBuildingTextureData))
  (if (> (Utilities|Array|Length _outtexturedataarray) 0)
    (bind _outtexturedataarray_2 (Building|GetBuildingTextureData))
    (CallFunction|SetBuildingTextureData 1 (Utilities|Array|Get(acopy) _outtexturedataarray_2))
    (Default|SetColoronSecondaryMesh (Math|Conversions|ToLinearColor(Color) (Variables|/Script/FortniteGame.BuildingTextureData|GetTintColor (Utilities|Array|Get(acopy) _outtexturedataarray_2))) _sm_rustbucket_smallsilo_a)
    (Rendering|Material|SetMaterial _sm_rustbucket_smallsilo_a 0 (Class|BTechArtGlobalLibrary|GetDynamicMaterialInstance (Variables|Debug|GetStaticMeshComponent)))))
